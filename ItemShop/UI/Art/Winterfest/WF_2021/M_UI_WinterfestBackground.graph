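# UI material for the Winterfest 2021 item shop background, layering animated snowflakes over a color gradient.
# The gradient samples a curve atlas row along the horizontal component of center-scaled, rotated UVs, with parameters for scale, angle and curve.
# Snowflakes use screen-aligned UVs corrected for aspect ratio, slowly rotated around an off-center pivot, scaled and panned over time.
# The snowflake texture's RGB is added to the gradient.
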Material /ItemShop/UI/Art/Winterfest/WF_2021/M_UI_WinterfestBackground
graph {
 properties {bUsedWithStaticMesh=false, materialDomain=MD_UI, pixelDepthOffsetMode=PDOM_Legacy, usedShadingModels=Default Lit}
 nodes [
  n1 = MaterialFunctionCall(Texture Scale=ScalarParameter() {defaultValue=1.5678, group=Gradient, parameterName=GradientScale}) {materialFunction=&ScaleUVsByCenter}
  n2 = ScalarParameter() {defaultValue=2.76, group=Gradient, parameterName=GradientAngle}
  n3 = CurveAtlasRowParameter(CurveTime=ComponentMask(None=MaterialFunctionCall(UVs=@n1.UVs, "Rotation Angle (0-1)"=@n2) {materialFunction=&CustomRotator}.Rotated Values) {r=true}) {
   atlas=/Game/UI/Frontend/ItemShop/Art/ItemShopColorAtlas
   curve=/Game/UI/Frontend/ItemShop/Art/Winterfest/WF_2021/ItemShop_WinterfestBG_Curve
   defaultValue=6
   group=Gradient
   parameterName=CurveParamsBG
  }
  n4 = MaterialFunctionCall(Float2=MaterialFunctionCall() {materialFunction=&ScreenResolution}.Visible Resolution) {materialFunction=&BreakOutFloat2Components}
  n5 = AppendVector(A=Divide(A=@n4.R, B=Max(A=@n4.G)), B=Constant() {r=1})
  n6 = MaterialFunctionCall(Texture Dimensions=TextureCoordinate()) {materialFunction=&ScreenAlignedUVs}
  n7 = Multiply(
   A=Rotator(Coordinate=Multiply(A=@n5, B=@n6."X 100%, Y 100%")) {centerX=5.0364, centerY=-1.7492, constCoordinate=1, speed=0.013}
   B=ScalarParameter() {defaultValue=3, parameterName=Scale}
  )
  n8 = Time()
  n9 = Constant() {r=0.032}
  Add(
   A=@n3
   B=TextureSample(UVs=Panner(Coordinate=@n7, Time=@n8, Speed=@n9)) {texture=&/T_UI_Winterfest_Snowflakes}.RGB
  )
 ]
}
