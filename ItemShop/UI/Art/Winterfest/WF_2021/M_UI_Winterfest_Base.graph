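# UI-domain material for Item Shop Winterfest 2021 art, built around red and green textures chosen by the static switch "Red".
# Each texture samples UVs scaled and centered from UI pixel size, but the ratios and comparison use a component against itself, leaving UVs effectively unchanged.
# The red branch is dimmed by a constant lerp; green passes straight through.
# Radial gradients, color parameters and extra texture lerps feed nothing shown, and no material output connection appears.
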Material /ItemShop/UI/Art/Winterfest/WF_2021/M_UI_Winterfest_Base
graph {
 properties {bUsedWithStaticMesh=false, materialDomain=MD_UI, pixelDepthOffsetMode=PDOM_Legacy, usedShadingModels=Default Lit}
 nodes [
  n1 = MaterialFunctionCall() {materialFunction=&GetUserInterfaceUV}
  n2 = MaterialFunctionCall(Float2=@n1.Pixel Size) {materialFunction=&BreakOutFloat2Components}
  n3 = Divide(A=@n2.G, B=@n2.G)
  n4 = AppendVector(A=Constant() {r=1}, B=@n3)
  n5 = Constant()
  n6 = Multiply(A=OneMinus(None=@n3)) {constB=0.5}
  n7 = MaterialFunctionCall(Float2=MaterialFunctionCall() {materialFunction=&GetUserInterfaceUV}.Pixel Size) {materialFunction=&BreakOutFloat2Components}
  n8 = Constant()
  n9 = MaterialFunctionCall() {materialFunction=&GetUserInterfaceUV}
  n10 = MaterialFunctionCall(Float2=@n9.Pixel Size) {materialFunction=&BreakOutFloat2Components}
  n11 = Divide(A=@n10.R, B=@n10.R)
  n12 = AppendVector(A=@n11, B=Constant() {r=1})
  n13 = Multiply(A=OneMinus(None=@n11)) {constB=0.5}
  n14 = Constant()
  n15 = LinearInterpolate(
   A=Add(A=Multiply(A=@n1.Normalized UV, B=@n4), B=AppendVector(A=@n5, B=@n6))
   B=Add(A=Multiply(A=@n9.Normalized UV, B=@n12), B=AppendVector(A=@n13, B=@n14))
   Alpha=If(A=@n7.R, B=@n7.R, "A > B"=@n8, "A == B"=@n8, "A < B"=Constant() {r=1})
  )
  LinearInterpolate(A=TextureSampleParameter2D(UVs=@n15) {parameterName=Winterfest_Red, texture=/OfferCatalog/Art/Winterfest/WF_2021/T_UI_WF_Red}.RGB) {constAlpha=0.04}
  n16 = MaterialFunctionCall() {materialFunction=&GetUserInterfaceUV}
  n17 = MaterialFunctionCall(Float2=@n16.Pixel Size) {materialFunction=&BreakOutFloat2Components}
  n18 = Divide(A=@n17.G, B=@n17.G)
  n19 = AppendVector(A=Constant() {r=1}, B=@n18)
  n20 = Constant()
  n21 = Multiply(A=OneMinus(None=@n18)) {constB=0.5}
  n22 = MaterialFunctionCall(Float2=MaterialFunctionCall() {materialFunction=&GetUserInterfaceUV}.Pixel Size) {materialFunction=&BreakOutFloat2Components}
  n23 = Constant()
  n24 = MaterialFunctionCall() {materialFunction=&GetUserInterfaceUV}
  n25 = MaterialFunctionCall(Float2=@n24.Pixel Size) {materialFunction=&BreakOutFloat2Components}
  n26 = Divide(A=@n25.R, B=@n25.R)
  n27 = AppendVector(A=@n26, B=Constant() {r=1})
  n28 = Multiply(A=OneMinus(None=@n26)) {constB=0.5}
  n29 = Constant()
  n30 = LinearInterpolate(
   A=Add(A=Multiply(A=@n16.Normalized UV, B=@n19), B=AppendVector(A=@n20, B=@n21))
   B=Add(A=Multiply(A=@n24.Normalized UV, B=@n27), B=AppendVector(A=@n28, B=@n29))
   Alpha=If(A=@n22.R, B=@n22.R, "A > B"=@n23, "A == B"=@n23, "A < B"=Constant() {r=1})
  )
  LinearInterpolate(A=TextureSampleParameter2D(UVs=@n30) {parameterName=Winterfest_Green, texture=/OfferCatalog/Art/Winterfest/WF_2021/T_UI_WF_Green}.RGB) {constAlpha=0.04}
  MaterialFunctionCall(
   UVs=MaterialFunctionCall() {materialFunction=&ScreenAlignedUVs}."X 100%, Y 100%"
   CenterPosition=Constant2Vector() {g=0.5, r=0.5}
   Radius=ScalarParameter() {defaultValue=0.5, parameterName=RadialRadius_02}
   Density=ScalarParameter() {defaultValue=1.0452, parameterName=RadialDensity_02}
  ) {materialFunction=&RadialGradientExponential}
  MaterialFunctionCall(
   UVs=MaterialFunctionCall() {materialFunction=&ScreenAlignedUVs}."X 100%, Y 100%"
   CenterPosition=Constant2Vector() {g=0.5, r=0.5}
   Radius=ScalarParameter() {defaultValue=0.5, parameterName=RadialRadius_02}
   Density=ScalarParameter() {defaultValue=1.0452, parameterName=RadialDensity_02}
  ) {materialFunction=&RadialGradientExponential}
  StaticSwitchParameter(
   True=LinearInterpolate(B=TextureSampleParameter2D(UVs=@n15) {parameterName=Winterfest_Red, texture=&T_UI_WF_Red}.RGB) {constAlpha=0.7}
   False=TextureSampleParameter2D(UVs=@n30) {parameterName=Winterfest_Green, texture=&T_UI_WF_Green}.RGB
  ) {parameterName=Red}
  VectorParameter() {parameterName=RGB2, defaultValue=<rgba 0 0 0.5469 1>}
  VectorParameter() {parameterName=RGB2, defaultValue=<rgba 0 0 0.5469 1>}
 ]
}
